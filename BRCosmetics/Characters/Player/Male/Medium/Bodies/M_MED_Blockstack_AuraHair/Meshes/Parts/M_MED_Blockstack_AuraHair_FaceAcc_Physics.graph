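# Physics asset for the face accessory mesh of the Blockstack AuraHair medium male cosmetic body, built around two dynamic ponytails.
# Each ponytail is a four-capsule chain from a base on the head, with twist locked and swing limited more tightly on one axis.
# Head and upper-spine boxes, plus clavicle and arm capsules, form the surrounding body shapes.
# These surrounding bodies and both ponytail bases link through fully free angular constraints in a tree rooted at spine_04.
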PhysicsAsset /BRCosmetics/Characters/Player/Male/Medium/Bodies/M_MED_Blockstack_AuraHair/Meshes/Parts/M_MED_Blockstack_AuraHair_FaceAcc_Physics
bodies {
 arm_bend_1_l [{shapeType=Capsule, length=10.5726, radius=4, shapeName=arm_bend_1_l_capsule, center=<6.3942 0 0>, rotation=<rot 90 0 0>}]
 arm_bend_1_r [{shapeType=Capsule, length=10.5726, radius=4, shapeName=arm_bend_1_r_capsule, center=<-6.3942 0 0>, rotation=<rot 90 180 0>}]
 arm_bend_3_l [{shapeType=Capsule, length=7.0312, radius=4, shapeName=arm_bend_3_l_capsule, center=<2.6731 0 0>, rotation=<rot 90 0 0>}]
 arm_bend_3_r [{shapeType=Capsule, length=7.0312, radius=4, shapeName=arm_bend_3_r_capsule, center=<-2.6731 -4.1761 0>, rotation=<rot 90 180 0>}]
 arm_bend_4_l [{shapeType=Capsule, length=7.0312, radius=4, shapeName=arm_bend_4_l_capsule, center=<2.6731 0 0>, rotation=<rot 90 0 0>}]
 arm_bend_4_r [{shapeType=Capsule, length=7.0312, radius=4, shapeName=arm_bend_4_r_capsule, center=<-2.6731 0 0>, rotation=<rot 90 180 0>}]
 base_ponytail_l [{shapeType=Capsule, length=0.505, radius=0.505, shapeName=base_ponytail_l_capsule, rotation=<rot 0 -90 0>}]
 base_ponytail_r [{shapeType=Capsule, length=0.505, radius=0.505, shapeName=base_ponytail_r_capsule, rotation=<rot 0 -90 -180>}]
 clavicle_l [{shapeType=Capsule, length=0.5, radius=0.5, shapeName=clavicle_l_capsule}]
 clavicle_r [{shapeType=Capsule, length=0.505, radius=0.505, shapeName=clavicle_r_capsule, rotation=<rot 0 -90 0>}]
 deform_head [{shapeType=Box, extentX=50.4675, extentY=33.2222, extentZ=26.1874, shapeName=deform_head_box, center=<-1.7767 0 0>}]
 dyn_ponytail_01_l [{shapeType=Capsule, length=7.3167, radius=4.5, shapeName=dyn_ponytail_01_l_capsule, center=<5.3385 0 0.6676>, rotation=<rot 90 0 0>}]
 dyn_ponytail_01_r [{shapeType=Capsule, length=7.3167, radius=4.5, shapeName=dyn_ponytail_01_r_capsule, center=<-5.3385 0 -0.6676>, rotation=<rot 90 180 0>}]
 dyn_ponytail_02_l [{shapeType=Capsule, length=7.6463, radius=4.5, shapeName=dyn_ponytail_02_l_capsule, center=<5.3385 0 0.3457>, rotation=<rot 90 0 0>}]
 dyn_ponytail_02_r [{shapeType=Capsule, length=7.6463, radius=4.5, shapeName=dyn_ponytail_02_r_capsule, center=<-5.3385 0 -0.3457>, rotation=<rot 90 180 0>}]
 dyn_ponytail_03_l [{shapeType=Capsule, length=6.875, radius=3.3205, shapeName=dyn_ponytail_03_l_capsule, center=<5.3385 0 0>, rotation=<rot 90 -17.7197 0>}]
 dyn_ponytail_03_r [{shapeType=Capsule, length=6.875, radius=3.3205, shapeName=dyn_ponytail_03_r_capsule, center=<-5.3385 0 0>, rotation=<rot 90 162.2803 0>}]
 dyn_ponytail_04_l [{shapeType=Capsule, length=4.5123, radius=3.0575, shapeName=dyn_ponytail_04_l_capsule, center=<3.8163 0 0>, rotation=<rot 90 0 0>}]
 dyn_ponytail_04_r [{shapeType=Capsule, length=4.5123, radius=3.0575, shapeName=dyn_ponytail_04_r_capsule, center=<-3.8163 0 0>, rotation=<rot 90 180 0>}]
 spine_04 [{
   shapeType=Box
   extentX=30.2044
   extentY=31.4031
   extentZ=28.3877
   shapeName=spine_04_box
   center=<-0.0832 1.1159 0>
   rotation=<rot 0 4.2659 0>
  }]
 upperRetArm_l_base [{shapeType=Capsule, length=5, radius=5, shapeName=upperRetArm_l_base_capsule, rotation=<rot 90 0 0>}]
 upperRetArm_r_base [{shapeType=Capsule, length=5, radius=5, shapeName=upperRetArm_r_base_capsule, rotation=<rot 90 180 0>}]
}
constraints [
 {
  bone1Name=dyn_ponytail_01_l
  bone2Name=base_ponytail_l
  swing1LimitDegrees=30
  swing2LimitDegrees=10
  twistLimitDegrees=45
  twistMotion=Locked
 }
 {
  bone1Name=dyn_ponytail_02_l
  bone2Name=dyn_ponytail_01_l
  swing1LimitDegrees=30
  swing2LimitDegrees=10
  twistLimitDegrees=45
  twistMotion=Locked
 }
 {
  bone1Name=dyn_ponytail_03_l
  bone2Name=dyn_ponytail_02_l
  swing1LimitDegrees=30
  swing2LimitDegrees=10
  twistLimitDegrees=45
  twistMotion=Locked
 }
 {
  bone1Name=dyn_ponytail_04_l
  bone2Name=dyn_ponytail_03_l
  swing1LimitDegrees=30
  swing2LimitDegrees=10
  twistLimitDegrees=45
  twistMotion=Locked
 }
 {
  bone1Name=base_ponytail_l
  bone2Name=deform_head
  swing1LimitDegrees=45
  swing1Motion=Free
  swing2LimitDegrees=45
  swing2Motion=Free
  twistLimitDegrees=45
  twistMotion=Free
 }
 {
  bone1Name=base_ponytail_r
  bone2Name=deform_head
  swing1LimitDegrees=45
  swing1Motion=Free
  swing2LimitDegrees=45
  swing2Motion=Free
  twistLimitDegrees=45
  twistMotion=Free
 }
 {
  bone1Name=dyn_ponytail_01_r
  bone2Name=base_ponytail_r
  swing1LimitDegrees=30
  swing2LimitDegrees=10
  twistLimitDegrees=45
  twistMotion=Locked
 }
 {
  bone1Name=dyn_ponytail_02_r
  bone2Name=dyn_ponytail_01_r
  swing1LimitDegrees=30
  swing2LimitDegrees=10
  twistLimitDegrees=45
  twistMotion=Locked
 }
 {
  bone1Name=dyn_ponytail_03_r
  bone2Name=dyn_ponytail_02_r
  swing1LimitDegrees=30
  swing2LimitDegrees=10
  twistLimitDegrees=45
  twistMotion=Locked
 }
 {
  bone1Name=dyn_ponytail_04_r
  bone2Name=dyn_ponytail_03_r
  swing1LimitDegrees=30
  swing2LimitDegrees=10
  twistLimitDegrees=45
  twistMotion=Locked
 }
 {
  bone1Name=arm_bend_1_l
  bone2Name=upperRetArm_l_base
  swing1LimitDegrees=45
  swing1Motion=Free
  swing2LimitDegrees=45
  swing2Motion=Free
  twistLimitDegrees=45
  twistMotion=Free
 }
 {
  bone1Name=arm_bend_3_l
  bone2Name=upperRetArm_l_base
  swing1LimitDegrees=45
  swing1Motion=Free
  swing2LimitDegrees=45
  swing2Motion=Free
  twistLimitDegrees=45
  twistMotion=Free
 }
 {
  bone1Name=arm_bend_4_l
  bone2Name=upperRetArm_l_base
  swing1LimitDegrees=45
  swing1Motion=Free
  swing2LimitDegrees=45
  swing2Motion=Free
  twistLimitDegrees=45
  twistMotion=Free
 }
 {
  bone1Name=arm_bend_1_r
  bone2Name=upperRetArm_r_base
  swing1LimitDegrees=45
  swing1Motion=Free
  swing2LimitDegrees=45
  swing2Motion=Free
  twistLimitDegrees=45
  twistMotion=Free
 }
 {
  bone1Name=arm_bend_3_r
  bone2Name=upperRetArm_r_base
  swing1LimitDegrees=45
  swing1Motion=Free
  swing2LimitDegrees=45
  swing2Motion=Free
  twistLimitDegrees=45
  twistMotion=Free
 }
 {
  bone1Name=arm_bend_4_r
  bone2Name=upperRetArm_r_base
  swing1LimitDegrees=45
  swing1Motion=Free
  swing2LimitDegrees=45
  swing2Motion=Free
  twistLimitDegrees=45
  twistMotion=Free
 }
 {
  bone1Name=deform_head
  bone2Name=spine_04
  swing1LimitDegrees=45
  swing1Motion=Free
  swing2LimitDegrees=45
  swing2Motion=Free
  twistLimitDegrees=45
  twistMotion=Free
 }
 {
  bone1Name=clavicle_l
  bone2Name=spine_04
  swing1LimitDegrees=45
  swing1Motion=Free
  swing2LimitDegrees=45
  swing2Motion=Free
  twistLimitDegrees=45
  twistMotion=Free
 }
 {
  bone1Name=clavicle_r
  bone2Name=spine_04
  swing1LimitDegrees=45
  swing1Motion=Free
  swing2LimitDegrees=45
  swing2Motion=Free
  twistLimitDegrees=45
  twistMotion=Free
 }
 {
  bone1Name=upperRetArm_l_base
  bone2Name=spine_04
  swing1LimitDegrees=45
  swing1Motion=Free
  swing2LimitDegrees=45
  swing2Motion=Free
  twistLimitDegrees=45
  twistMotion=Free
 }
 {
  bone1Name=upperRetArm_r_base
  bone2Name=spine_04
  swing1LimitDegrees=45
  swing1Motion=Free
  swing2LimitDegrees=45
  swing2Motion=Free
  twistLimitDegrees=45
  twistMotion=Free
 }
]
