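# Anim Blueprint for the AuraHair face accessory of the Blockstack male character, animating two ponytails.
# Copies the attached parent's pose through a Control Rig, offsets both ponytail bases, then simulates RigidBody physics in head-bone space.
# Post-physics bone rotations with bool-driven, eased blends bend both ponytails.
# The event graph computes gravity and rigid-body settings from a parameter asset, plus jump/fall, skydive and hang-glider states.
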AnimBlueprint /BRCosmetics/Characters/Player/Male/Medium/Bodies/M_MED_Blockstack_AuraHair/Meshes/Parts/M_MED_Blockstack_AuraHair_FaceAcc_AnimBP
target_skeleton=&/M_MED_Blockstack_AuraHair_FaceAcc_Skeleton
graphs [
 AnimGraph {nodes [
   n1 = Control Rig(Source=Copy Pose From Mesh() {bUseAttachedParent=true}) {controlRigAssetReference {controlRigAsset=&/M_MED_Blockstack_AuraHair_FaceAcc_CtrlRig}}
   n2 = "Transform (Modify) Bone - Bone: base_ponytail_l"(ComponentPose=Local To Component(LocalPose=@n1), Alpha=1, Translation="0,0,0", Rotation="0,-90,0", Scale="1,1,1") {
    rotationMode=BMM_Additive
    rotationSpace=BCS_BoneSpace
    translationMode=BMM_Additive
    translationSpace=BCS_BoneSpace
    alphaBoolBlend {blendInTime=0.2, blendOutTime=0.2}
    alphaScaleBiasClamp {bInterpResult=true}
    boneToModify {boneName=base_ponytail_l}
    rotation {yaw=-90}
   }
   n3 = "Transform (Modify) Bone - Bone: base_ponytail_r"(ComponentPose=@n2, Alpha=1, Translation="0,0,0", Rotation="0,-90,0", Scale="1,1,1") {
    rotationMode=BMM_Additive
    rotationSpace=BCS_BoneSpace
    translationMode=BMM_Additive
    translationSpace=BCS_BoneSpace
    alphaBoolBlend {blendInTime=0.2, blendOutTime=0.2}
    alphaScaleBiasClamp {bInterpResult=true}
    boneToModify {boneName=base_ponytail_r}
    rotation {yaw=-90}
   }
   n4 = RigidBody!(
    ComponentPose=@n3
    Alpha
    OverrideWorldGravity:Vector="0,0,-980"
    ExternalForce
    SimSpaceSettings:Sim Space Settings Structure=(WorldAlpha=0.7,VelocityScaleZ=1,DampingAlpha=1,MaxLinearVelocity=100,MaxAngularVelocity=15,MaxLinearAcceleration=200,MaxAngularAcceleration=50,ExternalLinearDragV=(X=0,Y=0,Z=0),ExternalLinearVelocity=(X=0,Y=0,Z=0),ExternalAngularVelocity=(X=0,Y=0,Z=0))
   ) -> (Pose) {
    bOverrideWorldGravity=true
    bUseDefaultAsSimulated=true
    overridePhysicsAsset=&/M_MED_Blockstack_AuraHair_FaceAcc_Physics
    simulationSpace=BaseBoneSpace
    baseBoneRef {boneName=head}
    componentAppliedLinearAccClamp=<0 0 0>
    overrideWorldGravity {z=-980}
    simSpaceSettings {maxAngularAcceleration=50, maxAngularVelocity=15, maxLinearAcceleration=200, maxLinearVelocity=100, worldAlpha=0.7}
   }
   n5 = "Transform (Modify) Bone - Bone: base_ponytail_r"!(ComponentPose=@n4, bAlphaBoolEnabled:bool=True, Translation="0,0,0", Rotation="15,0,0", Scale="1,1,1") -> (Pose) {
    alphaInputType=Bool
    rotationMode=BMM_Additive
    rotationSpace=BCS_BoneSpace
    translationMode=BMM_Additive
    translationSpace=BCS_BoneSpace
    alphaBoolBlend {blendInTime=1.3, blendOption=CubicInOut, blendOutTime=0.4}
    alphaScaleBiasClamp {bInterpResult=true, interpSpeedIncreasing=1}
    boneToModify {boneName=base_ponytail_r}
    rotation {pitch=15}
   }
   n6 = "Transform (Modify) Bone - Bone: dyn_ponytail_03_r"!(ComponentPose=@n5, bAlphaBoolEnabled:bool=True, Translation="0,0,0", Rotation="50,0,0", Scale="1,1,1") -> (Pose) {
    alphaInputType=Bool
    rotationMode=BMM_Additive
    rotationSpace=BCS_BoneSpace
    translationMode=BMM_Additive
    translationSpace=BCS_BoneSpace
    alphaBoolBlend {blendInTime=1.3, blendOption=CubicInOut, blendOutTime=0.4}
    alphaScaleBiasClamp {bInterpResult=true, interpSpeedIncreasing=2}
    boneToModify {boneName=dyn_ponytail_03_r}
    rotation {pitch=50}
   }
   n7 = "Transform (Modify) Bone - Bone: base_ponytail_l"!(ComponentPose=@n6, bAlphaBoolEnabled:bool=True, Translation="0,0,0", Rotation="15,0,0", Scale="1,1,1") -> (Pose) {
    alphaInputType=Bool
    rotationMode=BMM_Additive
    rotationSpace=BCS_BoneSpace
    translationMode=BMM_Additive
    translationSpace=BCS_BoneSpace
    alphaBoolBlend {blendInTime=1.3, blendOption=CubicInOut, blendOutTime=0.4}
    alphaScaleBiasClamp {bInterpResult=true, interpSpeedIncreasing=1}
    boneToModify {boneName=base_ponytail_l}
    rotation {pitch=15}
   }
   n8 = "Transform (Modify) Bone - Bone: dyn_ponytail_03_l"!(ComponentPose=@n7, bAlphaBoolEnabled:bool=True, Translation="0,0,0", Rotation="50,0,0", Scale="1,1,1") -> (Pose) {
    alphaInputType=Bool
    rotationMode=BMM_Additive
    rotationSpace=BCS_BoneSpace
    translationMode=BMM_Additive
    translationSpace=BCS_BoneSpace
    alphaBoolBlend {blendInTime=1.3, blendOption=CubicInOut, blendOutTime=0.4}
    alphaScaleBiasClamp {bInterpResult=true, interpSpeedIncreasing=2}
    boneToModify {boneName=dyn_ponytail_03_l}
    rotation {pitch=50}
   }
   Output Pose(Result=Component To Local(ComponentPose=@n8)) {layerGroup=None}
  ]}
 EventGraph {nodes [
   Event BlueprintUpdateAnimation()
   > n1 = "Gravity Override State Function (Parameter Asset) (Nativized)"(
    CharacterPartAnimInstance=Self-Reference()
    GameTimeInSeconds=GetGameTimeInSeconds()
    DebugJointName=spine_05
    Parameters=&/M_MED_Blockstack_AuraHair_FaceAcc_Parameters
   )
   > Set Gravity Override(Gravity Override=@n1)
   > Set RBAN Settings Out(RBAN Settings Out=@n1.OutRigidBodyAnimNodeParameters)
   n2 = Get bIsSkydiveFloating()
   n3 = Get bIsSkydiveDiveMode()
   n4 = Get HangGlidingAlpha() -> (HangGlidingAlpha:bool)
   > Set FloatingHangGlidingAndSkydiveMode(FloatingHangGlidingAndSkydiveMode:double="To Float (Boolean)"(InBool=OR Boolean!(A=@n2, B=@n3, C:bool=@n4) -> (ReturnValue))) -> (Output_Get:double)
   n5 = Get bIsSkyDiving()
   n6 = Get bIsJumping()
   n7 = Get bIsFalling()
   > Set JumpingComplete(JumpingComplete:bool=AND Boolean(A=OR Boolean!(A=@n7, B=@n6) -> (ReturnValue), B=NOT Boolean(A=@n5))) -> (Output_Get:bool)
   Event OnParachuteOpened()
   > Switch on EGliderType(Selection=Get PlayerGliderType() -> (PlayerGliderType=HangGlider))
   > Set HangGlidingAlpha(HangGlidingAlpha:bool=1) -> (Output_Get:bool)
   Event OnParachuteClosed()
   > Set HangGlidingAlpha(HangGlidingAlpha=false)
   NOT Boolean(A=Get bIsSkyDiving())
  ]}
]
variables [
 FloatingHangGlidingAndSkydiveMode {type=number}
 Gravity Override {type=Vector}
 HangGlidingAlpha {type=boolean}
 Head {type=SkeletalMeshComponentBudgeted}
 JumpingComplete {type=boolean}
 RBAN Settings Out {type=FortRigidBodyAnimNodeParameters}
]
